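# Niagara swing trail for the FrostIron Bass instrument used as a pickaxe, producing an icy, snowy streak.
# Two anim-trail ribbons, a longer-lived smoke trail and a short hard-edged streak, narrow and fade over their lifetime.
# Snow_Bits sprites spawn in probability-shaped bursts at a socket, inherit motion, and scatter with drag, gravity and curl noise.
# A system time-of-day color tints all three active emitters; two debug ribbon emitters are disabled.
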NiagaraSystem /SparksCosmetics/Instrument/Bass/FrostIron/Effects/NS_Pickaxe_Bass_FrostIron_Trail
dependencies {
 usedDataInterfaces [
  &NiagaraDataInterfaceCurve
  &NiagaraDataInterfaceColorCurve
  &NiagaraDataInterfaceVector2DCurve
  &NiagaraDataInterfaceSocketReader
 ]
 usedDynamicInputs []
 usedModules [
  &self:SystemSpawnScript.SystemScriptSource.SystemScriptGraph.NiagaraNodeAssignment_0.SetVariables_#1
  &self:SystemSpawnScript.SystemScriptSource.SystemScriptGraph.NiagaraNodeAssignment_1.SetVariables_#2
  &SystemState
  &BlendSpaceHelper
 ]
}
emitters {
 DebugRibbon {
  bEnabled=false
  emitterUpdateScript [
   &EmitterState()
   &SpawnBurst_Instantaneous(Spawn Count=2, Use Spawn Probability=false, UseLoopCountLimit=true)
  ]
  particleSpawnScript [&InitializeParticle(
    Color=<rgba 0.1754 0.2743 0.5755 0.3>
    Lifetime=$Emitter.CurrentLoopDuration
    Position Offset=&Lerp_Vector(A=<0 0 0>, B=<125 0 0>, Alpha=&ReturnNormalizedExecIndex())
    Ribbon Width=125
    Ribbon Width Mode=::NewEnumerator1
    Sprite Size Mode=::NewEnumerator3
    Uniform Sprite Size=1
    UsePositionOffset=true
    Write Material Random=true
   )]
  particleUpdateScript [
   &ParticleState(Kill Particles When Lifetime Has Elapsed=false)
   &SolveForcesAndVelocity(Clamp Velocity=false, Limit Acceleration=false)
  ]
  rendererClasses [&NiagaraSpriteRendererProperties, &NiagaraRibbonRendererProperties]
  renderers [
   {rendererClass=&NiagaraSpriteRendererProperties}
   {rendererClass=&NiagaraRibbonRendererProperties, rendererIndex=1}
  ]
 }
 DebugRibbon001 {
  bEnabled=false
  emitterUpdateScript [
   &EmitterState()
   &SpawnBurst_Instantaneous(Spawn Count=2, Use Spawn Probability=false, UseLoopCountLimit=true)
  ]
  particleSpawnScript [&InitializeParticle(
    Color=<rgba 0.2635 0.4872 0.9418 0.5>
    Lifetime=$Emitter.CurrentLoopDuration
    Position Offset=&Lerp_Vector(A=<0 0 0>, B=<125 0 0>, Alpha=&ReturnNormalizedExecIndex())
    Ribbon Width=100
    Ribbon Width Mode=::NewEnumerator1
    Sprite Size Mode=::NewEnumerator3
    Uniform Sprite Size=1
    UsePositionOffset=true
    Write Material Random=true
   )]
  particleUpdateScript [
   &ParticleState(Kill Particles When Lifetime Has Elapsed=false)
   &SolveForcesAndVelocity(Clamp Velocity=false, Limit Acceleration=false)
  ]
  rendererClasses [&NiagaraSpriteRendererProperties, &NiagaraRibbonRendererProperties]
  renderers [
   {rendererClass=&NiagaraSpriteRendererProperties}
   {rendererClass=&NiagaraRibbonRendererProperties, rendererIndex=1}
  ]
 }
 SnowSmokeTrail {
  emitterUpdateScript [
   &EmitterState(Life Cycle Mode=::NewEnumerator1, Loop Behavior=::NewEnumerator1, UseLoopDelay=false)
   &FN_SpawnAnimTrail_v2(AnimTrailWidthMode=::FromSecond)
  ]
  particleSpawnScript [
   /Game/Effects/Niagara/Modules/AnimTrail/FN_InitializeAnimTrail/FN_InitializeAnimTrailBoneIndecies(First Bone or Socket Index=1)
   /Game/Effects/Niagara/Modules/AnimTrail/FN_InitializeAnimTrail/FN_InitializeAnimTrail(AnimTrailWidth=1, Color=<rgba 0.1754 0.2743 0.5755 0.8>, Lifetime=0.7)
   &SystemLocation(MaskSpawnGroup=false)
  ]
  particleUpdateScript [
   &ParticleState()
   &ScaleColor(
    Scale Alpha=&Multiply_Float(
     A=&FloatFromCurve(FloatCurve=DataInterface() {propertyValues {Curve {keys [
          {arriveTangent=1.0519, interpMode=RCIM_Cubic, leaveTangent=1.0519, tangentMode=RCTM_User, time=0.05, value=0}
          {arriveTangentWeight=0.2733, interpMode=RCIM_Cubic, leaveTangentWeight=1.025, tangentMode=RCTM_User, time=0.25, value=0.7}
          {arriveTangentWeight=0.7981, interpMode=RCIM_Cubic, tangentMode=RCTM_User, time=1, value=0}
         ]}}})
     B=&FloatFromCurve(
      FloatCurve=DataInterface() {propertyValues {Curve {keys [<at 0 1>,<at 0.25 1>,<at 0.5 1>,<at 0.75 1>,<at 1 1>]}}}
      CurveIndex=Linked() {linkedVariable {name=NPC.FortniteNPC.TimeOfDay_0-1, type=NiagaraFloat}}
     )
    )
    Scale RGB=&MakeVectorFromLinearColorRGB(Color=$System.TOD Color)
    ScaleA=true
    ScaleRGB=true
   ) {module=ScaleColor002}
   &ScaleColor(
    Scale Alpha=&FloatFromCurve(FloatCurve=DataInterface() {propertyValues {Curve {keys [
         {arriveTangent=6.1048, interpMode=RCIM_Cubic, leaveTangent=6.1048, tangentMode=RCTM_User, time=0, value=0}
         <at 0.333 1 RCIM_Cubic>
         <at 1 0 RCIM_Cubic>
        ]}}})
    Scale RGB=<1 1.6 2>
    ScaleA=true
    ScaleRGB=true
   ) {enabled=false}
   &ScaleRibbonWidth(Ribbon Width Scale=&FloatFromCurve(FloatCurve=DataInterface() {propertyValues {Curve {keys [
         <at 0 0 RCIM_Cubic>
         {arriveTangent=-0.0511, interpMode=RCIM_Cubic, leaveTangent=-0.0511, tangentMode=RCTM_User, time=0.15, value=2}
         {arriveTangent=-2.3529, interpMode=RCIM_Cubic, leaveTangent=-2.3529, time=0.3, value=0.4954}
         <at 1 0 RCIM_Cubic>
        ]}}}))
   &DynamicMaterialParameters(
    Index 0 Param 1=4
    Index 0 Param 2=1
    Index 0 Param 3=1
    Index 0 Param 4=&FloatFromCurve(FloatCurve=DataInterface() {propertyValues {Curve {keys [{arriveTangent=-2.1594, interpMode=RCIM_Cubic, leaveTangent=-2.1594, tangentMode=RCTM_User, time=0, value=1}, <at 1 0>]}}})
    Module.Param0WriteEnabled=true
    Module.Param1WriteEnabled=true
    Module.Param2WriteEnabled=true
    Module.Param3WriteEnabled=true
   )
   &FN_UpdateAnimTrail_v2()
  ]
  rendererClasses [&NiagaraRibbonRendererProperties]
  renderers [{rendererClass=&NiagaraRibbonRendererProperties}]
 }
 Snow_Bits {
  emitterUpdateScript [
   &EmitterState(Life Cycle Mode=::NewEnumerator1, Loop Behavior=::NewEnumerator1, Loop Delay=0.01, Loop Duration=1, UseLoopDelay=true)
   &SpawnRate(
    Spawn Probability=&FloatFromCurve(
     FloatCurve=DataInterface() {propertyValues {Curve {
        defaultValue=1
        keys [
         <at 0 0 RCIM_Cubic>
         {arriveTangentWeight=0.1292, interpMode=RCIM_Cubic, leaveTangentWeight=0.2584, tangentMode=RCTM_User, time=0.1, value=1}
         {arriveTangentWeight=0.0667, interpMode=RCIM_Cubic, leaveTangentWeight=0.1, tangentMode=RCTM_User, time=0.3, value=1}
         <at 0.6 0 RCIM_Cubic>
        ]
       }}}
     CurveIndex=$Emitter.NormalizedLoopAge
    )
    SpawnRate=&RandomRangeFloat(Minimum=55, Maximum=200, Evaluation Type=::NewEnumerator1)
    Use Spawn Probability=true
   )
   &SpawnPerUnit(
    Max Movement Threshold=2000
    Module.Use Max Movement Threshold=true
    Module.Use Movement Tolerance=false
    Spawn Spacing=50
    Use Spawn Probability=true
   ) {enabled=false}
  ]
  particleSpawnScript [
   &InitializeParticle(
    Color=&ColorFromCurve(
     ColorCurve=DataInterface() {propertyValues {
       AlphaCurve {keys [<at 0 1 RCIM_Cubic>]}
       BlueCurve {keys [<at 0.1558 2 RCIM_Cubic>,<at 0.9351 1.5 RCIM_Cubic>]}
       GreenCurve {keys [<at 0.1558 2 RCIM_Cubic>,<at 0.9351 1.1672 RCIM_Cubic>]}
       RedCurve {keys [<at 0.1558 2 RCIM_Cubic>,<at 0.9351 0.7048 RCIM_Cubic>]}
      }}
     CurveIndex=&RandomRangeFloat(Minimum=0, Maximum=1)
    )
    Lifetime=&FloatFromCurve(
     FloatCurve=DataInterface() {propertyValues {Curve {keys [
         {interpMode=RCIM_Cubic, leaveTangentWeight=0.4612, tangentMode=RCTM_User, time=0.05, value=0.5}
         {arriveTangent=-0.1965, interpMode=RCIM_Cubic, leaveTangent=-0.1965, tangentMode=RCTM_User, time=0.5, value=0.25}
         <at 1 0.1>
        ]}}}
     CurveIndex=&RandomRangeFloat(Minimum=0, Maximum=1)
    )
    Sprite Rotation Mode=::NewEnumerator2
    Sprite Size=&Multiply_Vector2D_ByFloat(Float=&RandomRangeFloat(Minimum=2, Maximum=4), Vector2D=<1 1.5>)
    Sprite Size Mode=::NewEnumerator4
    UsePositionOffset=false
    Write Material Random=true
   )
   &SocketLocation(
    Interpolate Socket Sampling=true
    SamplePosOffset=true
    Sampled Position Offset=<0 0 15>
    Socket Index=1
    Socket Reader=$System.Socket Reader
    Socket Sampling Mode=::NewEnumerator3
    VelocitySpeedLimit=false
   )
   &ShapeLocation(MaskSpawnGroup=false, Override Seed=false, Sphere Radius=15, UseOwnerScale=true)
   &InheritVelocity(Inherited Velocity Amount Scale=<0.05 0.05 0.05>, Speed Limit=true, UseSpeedThreshold=false)
   &AddVelocity(Rotation Coordinate Space=::World, Velocity=$Output.SocketLocation.SampledVelocity, Velocity Speed Scale=0.07)
   &ApplyInitialForces()
   &DynamicMaterialParameters(
    Index 0 Param 1=3
    Index 0 Param 2=20
    Index 0 Param 3=1
    Module.Index0Float0Write=true
    Module.Index0Float1Write=true
    Module.Index0Float2Write=true
    Module.Index0Float3Write=true
   ) {module=DynamicMaterialParameters001}
  ]
  particleUpdateScript [
   &ParticleState()
   &ScaleSpriteSize(Uniform Curve Sprite Scale=DataInterface() {propertyValues {Curve {
       defaultValue=0
       keys [
        {interpMode=RCIM_Cubic, tangentMode=RCTM_User, time=0.1, value=1}
        {arriveTangent=-3.2192, interpMode=RCIM_Cubic, leaveTangent=-3.2192, tangentMode=RCTM_User, time=1, value=0}
       ]
      }}})
   &Drag(Drag=&RandomRangeFloat(Minimum=9, Maximum=15), Use Linear Drag=true, Use Rotational Drag=false)
   &GravityForce(Gravity=&Multiply_VectorByFloat(Vector=&RandomRangeVector(Minimum=<0 0 -100>, Maximum=<0 0 -300>), Float=1))
   &CurlNoiseForce(
    Bias Noise=false
    Noise Frequency=10
    Noise Strength=&FloatFromCurve(
     FloatCurve=DataInterface() {propertyValues {Curve {
        defaultValue=1
        keys [
         {arriveTangent=2.5569, interpMode=RCIM_Cubic, leaveTangent=2.5569, tangentMode=RCTM_User, time=0, value=0}
         {interpMode=RCIM_Cubic, tangentMode=RCTM_User, time=0.7, value=1}
        ]
       }}}
     Scale Curve=600
    )
    Pan Noise=false
   )
   &SolveForcesAndVelocity(Clamp Velocity=false, Limit Acceleration=false)
   &ScaleSpriteSizeBySpeed(Max Scale Factor=<0.5 1.5>, Min Scale Factor=<0.5 0.5>)
   &ScaleColor(
    Scale Alpha=&FloatFromCurve(FloatCurve=DataInterface() {propertyValues {Curve {
        defaultValue=1
        keys [{interpMode=RCIM_Cubic, tangentMode=RCTM_User, time=0.2, value=1}, {interpMode=RCIM_Cubic, tangentMode=RCTM_User, time=1, value=0}]
       }}})
    Scale RGB=&Multiply_Vector(
     A=&MakeVectorFromLinearColorRGB(Color=$System.TOD Color)
     B=&MakeVectorFromLinearColorRGB(Color=&ColorFromCurve(ColorCurve=DataInterface() {propertyValues {
         AlphaCurve {keys [<at 0.333 1>,<at 1 0>]}
         BlueCurve {keys [<at 0.0711 3>,<at 0.9615 2.256>]}
         GreenCurve {keys [<at 0.0711 3>,<at 0.9615 2.1654>]}
         RedCurve {keys [<at 0.0711 3>,<at 0.9615 0.4645>]}
        }}))
    )
    ScaleA=true
    ScaleRGB=true
   )
  ]
  rendererClasses [&NiagaraSpriteRendererProperties]
  renderers [{rendererClass=&NiagaraSpriteRendererProperties}]
 }
 TrailHardEdge {
  emitterUpdateScript [
   &EmitterState(Life Cycle Mode=::NewEnumerator1, Loop Behavior=::NewEnumerator1, Loop Duration=0.22, UseLoopDelay=true)
   &FN_SpawnAnimTrail_v2()
  ]
  particleSpawnScript [
   &FN_InitializeAnimTrailBoneIndecies(First Bone or Socket Index=1)
   &FN_InitializeAnimTrail(AnimTrailWidth=1, Color=<rgba 0.2635 0.4872 0.9418 0.5>, Lifetime=0.15)
   &SystemLocation(MaskSpawnGroup=false)
  ]
  particleUpdateScript [
   &ScaleRibbonWidth(Ribbon Width Scale=&FloatFromCurve(FloatCurve=DataInterface() {propertyValues {Curve {keys [
         {arriveTangent=-1.2579, interpMode=RCIM_Cubic, time=0, value=1}
         {arriveTangent=-0.9664, interpMode=RCIM_Cubic, leaveTangent=-0.9664, tangentMode=RCTM_User, time=1, value=0}
        ]}}}))
   &ScaleColor(
    Scale Alpha=&Multiply_Float(
     A=&FloatFromCurve(FloatCurve=DataInterface() {propertyValues {Curve {keys [
          {interpMode=RCIM_Cubic, leaveTangent=-4.8418, tangentMode=RCTM_Break, time=0, value=1}
          <at 0.5 0.1>
          {arriveTangent=-0.2139, interpMode=RCIM_Cubic, tangentMode=RCTM_Break, time=1, value=0}
         ]}}})
     B=&FloatFromCurve(
      FloatCurve=DataInterface() {propertyValues {Curve {keys [<at 0 1>,<at 0.25 1>,<at 0.5 1>,<at 0.75 1>,<at 1 1>]}}}
      CurveIndex=Linked() {linkedVariable {name=NPC.FortniteNPC.TimeOfDay_0-1, type=NiagaraFloat}}
     )
    )
    Scale RGB=&MakeVectorFromLinearColorRGB(Color=$System.TOD Color)
    ScaleA=true
    ScaleRGB=true
   ) {module=ScaleColor002}
   &FN_UpdateAnimTrail_v2()
   &CameraOffset(Camera Offset Amount=10) {enabled=false}
  ]
  rendererClasses [&NiagaraRibbonRendererProperties]
  renderers [{rendererClass=&NiagaraRibbonRendererProperties}]
 }
}
summary {
 systemName=NS_Pickaxe_Bass_FrostIron_Trail
 emitters [
  {
   bEnabled=false
   emitterName=DebugRibbon
   rendererClasses [&NiagaraSpriteRendererProperties, &NiagaraRibbonRendererProperties]
  }
  {emitterName=SnowSmokeTrail, rendererClasses [&NiagaraRibbonRendererProperties]}
  {emitterName=TrailHardEdge, rendererClasses [&NiagaraRibbonRendererProperties]}
  {emitterName=Snow_Bits, rendererClasses [&NiagaraSpriteRendererProperties]}
  {
   bEnabled=false
   emitterName=DebugRibbon001
   rendererClasses [&NiagaraSpriteRendererProperties, &NiagaraRibbonRendererProperties]
  }
 ]
 userVariables [
  User.bIsFrontEnd {type=NiagaraBool}
  User.bIsFrontEndPreview {type=NiagaraBool}
  User.IsVariant {type=NiagaraFloat, defaultValue=0, description=""}
  User.RibbonIDOffset {type=NiagaraInt32}
  User.SkeletalMesh {
   type=NiagaraDataInterfaceSkeletalMesh
   defaultValue {
    struct=NiagaraExt_VariableValue_DataInterface
    value {dataInterface=&self:NiagaraDataInterfaceSkeletalMesh_0}
   }
  }
 ]
}
